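# Looping ambient emitter cue for a bar tap in the Cosmos biome's Watering Hole, in the indoor ambient sound class.
# A looping drink machine wave player feeds a Modulator at default settings.
# Overridden attenuation keeps the sound local, with occlusion that lowers volume and applies a low-pass filter.
# Interpolated listener focus uses a narrow focus cone and zero non-focus volume, so the sound fades out as the listener turns away.
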SoundCue /Game/Sounds/Apollo/Biomes/Cosmos/AMB_Emit_Cosmos_WateringHole_BarTap_01_Loop_Cue
root=Modulator_0
nodes [
 {id=Modulator_0, children [WavePlayer_2]}
 {
  id=WavePlayer_2
  properties {bLooping=true, soundWaveAssetPtr=&/AMB_Cosmos_WateringHole_Bar_DrinkMachine_Loop_01}
 }
]
properties {
 attenuationSettings=/Game/Sounds/Attenuation/Fort_TileAmbient_Att
 bOverrideAttenuation=true
 maxDistance=768
 soundClassObject=&Fort_Ambient_Indoors_Athena
 volumeMultiplier=1.002
 attenuationOverrides {
  bEnableFocusInterpolation=true
  bEnableListenerFocus=true
  bEnableOcclusion=true
  falloffDistance=512
  focusAttackInterpSpeed=2
  focusAzimuth=20
  focusReleaseInterpSpeed=2
  nonFocusVolumeAttenuation=0
  occlusionInterpolationTime=1
  occlusionLowPassFilterFrequency=2000
  occlusionTraceChannel=ECC_GameTraceChannel11
  occlusionVolumeAttenuation=0.5
  attenuationShapeExtents {x=256}
 }
}
