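# Ambient emitter SoundCue for the Cosmos biome watering hole beacon, playing as a positional loop in the Athena ambient sound class.
# A mixer layers two looping waves, a beacon loop and a flag loop, each through its own modulator.
# Overridden attenuation adds distance falloff, listener focus and occlusion that both lowers volume and low-pass filters the sound.
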SoundCue /Game/Sounds/Apollo/Biomes/Cosmos/AMB_Emit_Cosmos_WateringHole_Beacon_01_Loop_Cue
root=Mixer_0
nodes [
 {id=Mixer_0, children [Modulator_0,Modulator_1], properties {inputVolume [1,1]}}
 {id=Modulator_0, children [WavePlayer_0]}
 {id=Modulator_1, children [WavePlayer_1]}
 {id=WavePlayer_0, properties {bLooping=true, soundWaveAssetPtr=&/AMB_Cosmos_WateringHole_Beacon_Loop_01}}
 {id=WavePlayer_1, properties {bLooping=true, soundWaveAssetPtr=&/AMB_Cosmos_WateringHole_Flag_Loop_01}}
]
properties {
 attenuationSettings=/Game/Sounds/Attenuation/Fort_TileAmbient_Att
 bOverrideAttenuation=true
 maxDistance=1500
 soundClassObject=&Fort_Ambient_Athena
 volumeMultiplier=1.002
 attenuationOverrides {
  bEnableListenerFocus=true
  bEnableOcclusion=true
  falloffDistance=1500
  nonFocusVolumeAttenuation=0.5
  occlusionInterpolationTime=1
  occlusionLowPassFilterFrequency=2000
  occlusionTraceChannel=ECC_GameTraceChannel11
  occlusionVolumeAttenuation=0.5
  attenuationShapeExtents {x=0}
 }
 concurrencyOverrides {maxCount=1}
}
