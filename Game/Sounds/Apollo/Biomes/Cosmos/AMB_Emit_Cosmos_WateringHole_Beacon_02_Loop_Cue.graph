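# Sound cue for a looping ambient emitter on the Cosmos biome's Watering Hole beacon, variant 02.
# A looping wave player feeds a Modulator into the output at reduced volume on the Athena ambient sound class.
# It references the tile ambient attenuation asset but overrides it with a short falloff, listener focus and trace-based occlusion that lowers volume and low-pass filters blocked sound.
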SoundCue /Game/Sounds/Apollo/Biomes/Cosmos/AMB_Emit_Cosmos_WateringHole_Beacon_02_Loop_Cue
root=Modulator_0
nodes [
 {id=Modulator_0, children [WavePlayer_1]}
 {id=WavePlayer_1, properties {bLooping=true, soundWaveAssetPtr=&/AMB_Cosmos_WateringHole_Beacon_Loop_02}}
]
properties {
 attenuationSettings=/Game/Sounds/Attenuation/Fort_TileAmbient_Att
 bOverrideAttenuation=true
 maxDistance=1500
 soundClassObject=&Fort_Ambient_Athena
 volumeMultiplier=0.752
 attenuationOverrides {
  bEnableListenerFocus=true
  bEnableOcclusion=true
  falloffDistance=1500
  nonFocusVolumeAttenuation=0.5
  occlusionLowPassFilterFrequency=1000
  occlusionTraceChannel=ECC_GameTraceChannel11
  occlusionVolumeAttenuation=0.5
  attenuationShapeExtents {x=0}
 }
}
