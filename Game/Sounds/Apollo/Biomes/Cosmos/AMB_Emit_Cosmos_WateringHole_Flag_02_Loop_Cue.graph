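# Looping ambient emitter cue for flags at the Cosmos biome watering hole, classed as Athena outdoor ambience.
# A Random node picks one of two looping flag waves at equal weight and feeds a Modulator at the root.
# Attenuation overrides replace the assigned tile-ambient asset, adding occlusion with low-pass filtering and volume reduction, plus listener focus that quiets the sound when out of focus.
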SoundCue /Game/Sounds/Apollo/Biomes/Cosmos/AMB_Emit_Cosmos_WateringHole_Flag_02_Loop_Cue
root=Modulator_0
nodes [
 {id=Modulator_0, children [Random_0]}
 {id=Random_0, children [WavePlayer_2,WavePlayer_1], properties {weights [1,1]}}
 {id=WavePlayer_1, properties {bLooping=true, soundWaveAssetPtr=&/AMB_Cosmos_WateringHole_Flag_Loop_02}}
 {id=WavePlayer_2, properties {bLooping=true, soundWaveAssetPtr=&/AMB_Cosmos_WateringHole_Flag_Loop_01}}
]
properties {
 attenuationSettings=/Game/Sounds/Attenuation/Fort_TileAmbient_Att
 bOverrideAttenuation=true
 maxDistance=1200
 soundClassObject=&Fort_Ambient_Outdoors_Athena
 volumeMultiplier=0.877
 attenuationOverrides {
  bEnableListenerFocus=true
  bEnableOcclusion=true
  falloffDistance=1200
  nonFocusVolumeAttenuation=0.5
  occlusionLowPassFilterFrequency=1000
  occlusionTraceChannel=ECC_GameTraceChannel11
  occlusionVolumeAttenuation=0.5
  attenuationShapeExtents {x=0}
 }
}
