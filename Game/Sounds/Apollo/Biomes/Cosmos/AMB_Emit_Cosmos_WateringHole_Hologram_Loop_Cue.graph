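# Looping ambient emitter cue for the hologram projector at the Cosmos Watering Hole, routed to the Athena ambient sound class with boosted volume.
# A looping wave player feeds a Modulator left at its default settings.
# Attenuation overrides replace the referenced tile-ambient asset with a short falloff from a point.
# Listener focus uses a narrow azimuth and zero non-focus volume, so the hum fades out when outside the listener's focus.
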SoundCue /Game/Sounds/Apollo/Biomes/Cosmos/AMB_Emit_Cosmos_WateringHole_Hologram_Loop_Cue
root=Modulator_0
nodes [
 {id=Modulator_0, children [WavePlayer_0]}
 {id=WavePlayer_0, properties {bLooping=true, soundWaveAssetPtr=&/AMB_Cosmos_WateringHole_Holo_Proj_Loop_01}}
]
properties {
 attenuationSettings=/Game/Sounds/Attenuation/Fort_TileAmbient_Att
 bOverrideAttenuation=true
 maxDistance=1000
 soundClassObject=&Fort_Ambient_Athena
 volumeMultiplier=2.005
 attenuationOverrides {
  bEnableFocusInterpolation=true
  bEnableListenerFocus=true
  falloffDistance=1000
  focusAttackInterpSpeed=2
  focusAzimuth=20
  focusReleaseInterpSpeed=2
  nonFocusVolumeAttenuation=0
  occlusionInterpolationTime=1
  occlusionLowPassFilterFrequency=2000
  occlusionVolumeAttenuation=0.5
  attenuationShapeExtents {x=0}
 }
}
