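# Ambient sound cue for the Cosmos biome crash area, playing intermittent metal stress sounds in the outdoor ambience class.
# A loop repeatedly waits a random delay of a few seconds, then picks one of seven equally weighted metal stress waves.
# A modulator at the root varies volume and slightly lowers pitch.
# Overridden attenuation adds distance-based low-pass filtering and wide stereo spread.
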SoundCue /Game/Sounds/Apollo/Biomes/Cosmos/AMB_SS_S15_Cosmos_Crash_Metal_Stres_Cue
root=Modulator_0
nodes [
 {id=Delay_0, children [Random_1], properties {delayMax=8, delayMin=1}}
 {id=Looping_0, children [Delay_0]}
 {id=Modulator_0, children [Looping_0], properties {pitchMax=1, pitchMin=0.9, volumeMax=1, volumeMin=0.5}}
 {
  id=Random_1
  children [WavePlayer_3,WavePlayer_4,WavePlayer_5,WavePlayer_6,WavePlayer_7,WavePlayer_8,WavePlayer_9]
  properties {weights [1,1,1,1,1,1,1]}
 }
 {id=WavePlayer_3, properties {soundWaveAssetPtr=&/AMB_SS_S15_Cosmos_MetalStress_01}}
 {id=WavePlayer_4, properties {soundWaveAssetPtr=&/AMB_SS_S15_Cosmos_MetalStress_02}}
 {id=WavePlayer_5, properties {soundWaveAssetPtr=&/AMB_SS_S15_Cosmos_MetalStress_03}}
 {id=WavePlayer_6, properties {soundWaveAssetPtr=&/AMB_SS_S15_Cosmos_MetalStress_04}}
 {id=WavePlayer_7, properties {soundWaveAssetPtr=&/AMB_SS_S15_Cosmos_MetalStress_05}}
 {id=WavePlayer_8, properties {soundWaveAssetPtr=&/AMB_SS_S15_Cosmos_MetalStress_06}}
 {id=WavePlayer_9, properties {soundWaveAssetPtr=&/AMB_SS_S15_Cosmos_MetalStress_07}}
]
properties {
 attenuationSettings=/Game/Sounds/Apollo/Biomes/Interior_Banks/AmbientAudio_Exterior_Att
 bOverrideAttenuation=true
 maxDistance=2500
 soundClassObject=&Fort_Ambient_Outdoors_Athena
 volumeMultiplier=2.005
 attenuationOverrides {
  bAttenuateWithLPF=true
  falloffDistance=2500
  lPFFrequencyAtMax=1000
  lPFFrequencyAtMin=10000
  lPFRadiusMax=1500
  lPFRadiusMin=800
  stereoSpread=1000
  attenuationShapeExtents {x=0}
 }
}
