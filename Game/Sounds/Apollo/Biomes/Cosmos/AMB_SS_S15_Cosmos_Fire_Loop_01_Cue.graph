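# Sound cue for a looping fire ambience in the Cosmos biome, filed under Apollo biome sounds.
# A single looping wave player feeds the output directly, playing the Loop_02 wave despite the cue's Loop_01 name.
# It uses the outdoor ambient sound class and a boosted volume.
# Although it references a shared exterior ambient attenuation asset, overridden attenuation is enabled, with occlusion tracing and a custom falloff.
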SoundCue /Game/Sounds/Apollo/Biomes/Cosmos/AMB_SS_S15_Cosmos_Fire_Loop_01_Cue
root=WavePlayer_2
nodes [{id=WavePlayer_2, properties {bLooping=true, soundWaveAssetPtr=&/AMB_SS_S15_Cosmos_Fire_Loop_02}}]
properties {
 attenuationSettings=/Game/Sounds/Apollo/Biomes/Interior_Banks/AmbientAudio_Exterior_Att
 bOverrideAttenuation=true
 maxDistance=1500
 soundClassObject=&Fort_Ambient_Outdoors_Athena
 volumeMultiplier=1.504
 attenuationOverrides {bEnableOcclusion=true, falloffDistance=1500, occlusionTraceChannel=ECC_GameTraceChannel11, attenuationShapeExtents {x=0}}
}
