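# Sound cue for a small looping fire ambience in the Season 15 Cosmos biome.
# A single looping wave player feeds the root directly, playing the Cosmos fire loop wave with no randomization or mixing.
# It routes to the outdoor ambient sound class, uses a shared exterior ambient attenuation asset, and carries a slight volume boost.
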SoundCue /Game/Sounds/Apollo/Biomes/Cosmos/AMB_SS_S15_Cosmos_Fire_SmLoop_01_Cue1
root=WavePlayer_0
nodes [{id=WavePlayer_0, properties {bLooping=true, soundWaveAssetPtr=&/AMB_SS_S15_Cosmos_Fire_Loop_01}}]
properties {
 attenuationSettings=/Game/Sounds/Apollo/Biomes/Interior_Banks/AmbientAudio_Exterior_Att
 maxDistance=3000
 soundClassObject=&Fort_Ambient_Outdoors_Athena
 volumeMultiplier=1.253
}
