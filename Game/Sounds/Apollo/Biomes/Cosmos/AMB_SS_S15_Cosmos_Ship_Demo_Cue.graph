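# Sound cue for the Cosmos biome ship demolition, playing a metal destruction impact in the Fort_Impacts sound class.
# A single wave player feeds a Modulator at the root, which applies pitch and volume variation at its default ranges.
# The cue references the shared medium-object demolition attenuation but overrides it with its own settings, boosts volume, and uses a four-max object-impact concurrency set.
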SoundCue /Game/Sounds/Apollo/Biomes/Cosmos/AMB_SS_S15_Cosmos_Ship_Demo_Cue
root=Modulator_2
nodes [
 {id=Modulator_2, children [WavePlayer_0]}
 {id=WavePlayer_0, properties {soundWaveAssetPtr=&/MTL_Demo_Cosmos_Ship_01}}
]
properties {
 attenuationSettings=/Game/Sounds/Attenuation/Fort_Demolition_Destroy_Medium_Sized_Object_Att
 bOverrideAttenuation=true
 maxDistance=3300
 soundClassObject=&Fort_Impacts
 volumeMultiplier=1.6
 attenuationOverrides {falloffDistance=2500, stereoSpread=1000, attenuationShapeExtents {x=800}}
 concurrencyOverrides {maxCount=4}
 concurrencySet [&CONC_Object_Impacts_FourMax]
}
