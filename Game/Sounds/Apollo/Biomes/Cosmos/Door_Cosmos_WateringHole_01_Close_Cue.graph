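# SoundCue for the closing sound of the Cosmos WateringHole door 01 in the Apollo biome sounds.
# A single wave player node plays the Door_Cosmos_WateringHole_Close_01 wave directly as the root, with no randomization or mixing.
# It uses the shared Fort_Doors attenuation and Fort_Doors sound class for spatial falloff and grouping.
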SoundCue /Game/Sounds/Apollo/Biomes/Cosmos/Door_Cosmos_WateringHole_01_Close_Cue
root=WavePlayer_1
nodes [{id=WavePlayer_1, properties {soundWaveAssetPtr=&/Door_Cosmos_WateringHole_Close_01}}]
properties {
 attenuationSettings=/Game/Sounds/Attenuation/Fort_Doors_Att
 maxDistance=2400
 soundClassObject=&Fort_Doors
 volumeMultiplier=1
}
